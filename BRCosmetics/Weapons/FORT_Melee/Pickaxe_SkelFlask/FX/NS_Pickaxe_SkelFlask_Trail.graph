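# Niagara swing trail for the SkelFlask pickaxe cosmetic, with color and intensity switched by cosmetic variant.
# Glow, main, tip and variant ribbon emitters use shared FN anim-trail modules between bone or socket indices, faded by a per-swing value.
# A socket-sampled sprite and ribbon emitter adds drifting curl-noise particles for some variants, with distance-scaled spawning.
# System-level time-of-day color and float values tint and dim the layers.
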NiagaraSystem /BRCosmetics/Weapons/FORT_Melee/Pickaxe_SkelFlask/FX/NS_Pickaxe_SkelFlask_Trail
dependencies {
 usedDataInterfaces [
  &NiagaraDataInterfaceCurve
  &NiagaraDataInterfaceArrayColor
  /Script/Niagara.NiagaraDataInterfaceArrayBool
  &NiagaraDataInterfaceColorCurve
  &NiagaraDataInterfaceArrayInt32
  &NiagaraDataInterfaceSocketReader
 ]
 usedDynamicInputs []
 usedModules [
  &self:SystemSpawnScript.SystemScriptSource.SystemScriptGraph.NiagaraNodeAssignment_0.SetVariables_#1
  &SystemState
  &FN_VariantManager
  &self:CrewVariantManager
  &self:SystemSpawnScript.SystemScriptSource.SystemScriptGraph.NiagaraNodeAssignment_2.SetVariables_#2
  &self:SystemSpawnScript.SystemScriptSource.SystemScriptGraph.NiagaraNodeAssignment_3.SetVariables_#3
  &self:SystemSpawnScript.SystemScriptSource.SystemScriptGraph.NiagaraNodeAssignment_7.SetVariables_#4
  &self:SystemSpawnScript.SystemScriptSource.SystemScriptGraph.NiagaraNodeAssignment_6.SetVariables_#5
 ]
}
emitters {
 Glow_Trail {
  emitterUpdateScript [
   &EmitterState(Life Cycle Mode=::NewEnumerator1, Loop Behavior=::NewEnumerator1, Loop Duration=0.27, UseLoopDelay=true)
   &FN_SpawnAnimTrail()
  ]
  particleSpawnScript [
   &FN_InitializeAnimTrailBoneIndecies(First Bone or Socket Index=2, Second Bone or Socket Index=2)
   &FN_InitializeAnimTrail(
    AnimTrailWidth=25
    Color=&VariantColorSelector(
     Variant Amount=4
     Original Color=<rgba 0.5213 0.8408 1 1>
     Variant 1 Color=<rgba 0.7545 0.9976 1.2 1>
     Variant 2 Color=<rgba 1.5 1.3346 0.9261 1>
     Variant 3 Color=&RandomRangeLinearColor(Minimum=<rgba 0.4566 1 0.8845 1>, Maximum=<rgba 0.941 0.6415 1 1>)
     Variant 4 Color=<rgba 0.15 0.575 1 1>
     Variant 5 Color=<rgba 0.7282 0.2 1 1>
    )
    Lifetime=0.17
   )
  ]
  particleUpdateScript [
   &FN_UpdateAnimTrail() {module=FN_UpdateAnimTrail001}
   &SolveForcesAndVelocity(Clamp Velocity=false, Limit Acceleration=false)
   &DynamicMaterialParameters(
    Index 1 Param 3=&FloatFromCurve(
     FloatCurve=DataInterface() {propertyValues {Curve {keys [<at 0 0>, {arriveTangent=1.9814, tangentMode=RCTM_User, time=1, value=1}]}}}
     CurveIndex=&ModuloFloat(A=$Emitter.Age, B=1)
     Scale Curve=0
    )
    Index 1 Param 4=1
    Index 2 Param 1=20
    Index 2 Param 2=1
    Index 2 Param 3=0.5
    Index 2 Param 4=0.6
    Module.Index0Float0Write=true
    Module.Index0Float1Write=true
    Module.Index0Float2Write=true
    Module.Index0Float3Write=true
    Module.Index1Float0Write=true
    Module.Index1Float1Write=true
    Module.Index1Float2Write=true
    Module.Index1Float3Write=true
    Module.Index2Float0Write=true
    Module.Index2Float1Write=true
    Module.Index2Float2Write=true
    Module.Index2Float3Write=true
    Write Parameter Index 1=::NewEnumerator1
    Write Parameter Index 2=::NewEnumerator1
   )
   &ScaleColor(
    Scale Alpha=&Multiply_Float(
     A=&FloatFromCurve(
      FloatCurve=DataInterface() {propertyValues {Curve {keys [
          {interpMode=RCIM_Cubic, leaveTangent=-1.6582, tangentMode=RCTM_Break, time=0, value=0.7}
          <at 0.7066 0>
          {arriveTangent=-0.2139, interpMode=RCIM_Cubic, tangentMode=RCTM_Break, time=1, value=0}
         ]}}}
      Scale Curve=$System.PerSwingFade
     )
     B=&Multiply_Float(A=$System.TODFloat, B=1)
    )
    Scale RGB=&Multiply_Vector(
     A=&MakeVectorFromLinearColorRGB(Color=$System.TODColor)
     B=&VectorFromFloat(Value=3)
    )
    ScaleA=true
    ScaleRGB=true
   )
  ]
  rendererClasses [&NiagaraRibbonRendererProperties]
  renderers [{rendererClass=&NiagaraRibbonRendererProperties}]
 }
 Main_Trail {
  emitterUpdateScript [
   &EmitterState(Life Cycle Mode=::NewEnumerator1, Loop Behavior=::NewEnumerator1, Loop Duration=0.27, UseLoopDelay=true)
   &FN_SpawnAnimTrail()
  ]
  particleSpawnScript [
   &FN_InitializeAnimTrailBoneIndecies()
   &FN_InitializeAnimTrail(
    Color=&VariantColorSelector(
     Variant Amount=4
     BlendAmount=0
     Use TODCurve?=true
     Original Color=<rgba 0.05 0.05 0.05 1>
     Variant 1 Color=<rgba 0.05 0.05 0.05 1>
     Variant 2 Color=<rgba 0 0 0 0.5>
     Variant 3 Color=&RandomRangeLinearColor(Minimum=<rgba 0.4566 1 0.8845 0.5>, Maximum=<rgba 0.941 0.6415 1 0.5>)
     Variant 4 Color=<rgba 0.15 0.575 1 1>
     Variant 5 Color=<rgba 0 0 0 0>
    )
    Lifetime=0.22
   )
  ]
  particleUpdateScript [
   &FN_UpdateAnimTrail() {module=FN_UpdateAnimTrail001}
   &SolveForcesAndVelocity(Clamp Velocity=false, Limit Acceleration=false)
   &DynamicMaterialParameters(
    Index 1 Param 3=&FloatFromCurve(
     FloatCurve=DataInterface() {propertyValues {Curve {keys [<at 0 0>, {arriveTangent=1.9814, tangentMode=RCTM_User, time=1, value=1}]}}}
     CurveIndex=&ModuloFloat(A=$Emitter.Age, B=1)
     Scale Curve=0
    )
    Index 1 Param 4=1
    Index 2 Param 1=20
    Index 2 Param 2=1
    Index 2 Param 3=0.5
    Index 2 Param 4=0.6
    Module.Index0Float0Write=true
    Module.Index0Float1Write=true
    Module.Index0Float2Write=true
    Module.Index0Float3Write=true
    Module.Index1Float0Write=true
    Module.Index1Float1Write=true
    Module.Index1Float2Write=true
    Module.Index1Float3Write=true
    Module.Index2Float0Write=true
    Module.Index2Float1Write=true
    Module.Index2Float2Write=true
    Module.Index2Float3Write=true
    Write Parameter Index 1=::NewEnumerator1
    Write Parameter Index 2=::NewEnumerator1
   )
   &ScaleColor(
    Scale Alpha=&Multiply_Float(
     A=&FloatFromCurve(
      FloatCurve=DataInterface() {propertyValues {Curve {keys [
          {interpMode=RCIM_Cubic, leaveTangent=-1.6582, tangentMode=RCTM_Break, time=0, value=0.7}
          <at 0.7066 0>
          {arriveTangent=-0.2139, interpMode=RCIM_Cubic, tangentMode=RCTM_Break, time=1, value=0}
         ]}}}
      Scale Curve=$System.PerSwingFade
     )
     B=&Multiply_Float(
      A=$System.TODFloat
      B=&MergeHelper_FloatSwitch(
       0=2
       1=2
       2=2
       3=0.5
       4=0.75
       5=2
       Index=&MakeIntFromFloat(Float=$System.VariantIndex)
       Num=::NewEnumerator4
      )
     )
    )
    Scale RGB=&Multiply_Vector(
     A=&VectorFromFloat(Value=$System.TODFloat)
     B=&MakeVectorFromLinearColorRGB(Color=$System.TODColor)
    )
    ScaleA=true
    ScaleRGB=true
   )
   /SetVariables(Particles.Lines Color:LinearColor=&MultiplyLinearColor(
     Linear Color=&VariantColorSelector(
      Variant Amount=4
      Original Color=<rgba 0.1205 0.1262 0.1337 1>
      Variant 1 Color=<rgba 0.1958 0.1945 0.2066 1>
      Variant 2 Color=<rgba 0.3073 0.2113 0.0768 0.5>
      Variant 3 Color=<rgba 0.4566 1 0.8845 1>
      Variant 4 Color=<rgba 0.15 0.575 1 1>
      Variant 5 Color=<rgba 0.0339 0.0467 0.0885 0.5>
     )
     Scale Factor=&MakeVector4FromVectorAndFloat(
      "Vector (RGB)"=&Multiply_Vector(
       A=&VectorFromFloat(Value=$System.TODFloat)
       B=&MakeVectorFromLinearColorRGB(Color=$System.TODColor)
      )
      "Float (Alpha)"=&Multiply_Float(
       A=&FloatFromCurve(
        FloatCurve=DataInterface() {propertyValues {Curve {keys [
            {interpMode=RCIM_Cubic, leaveTangent=-1.6582, tangentMode=RCTM_Break, time=0, value=0.7}
            <at 0.7066 0>
            {arriveTangent=-0.2139, interpMode=RCIM_Cubic, tangentMode=RCTM_Break, time=1, value=0}
           ]}}}
        Scale Curve=$System.PerSwingFade
       )
       B=&Multiply_Float(
        A=$System.TODFloat
        B=&MergeHelper_FloatSwitch(0=0, 1=0, 2=0, 3=0, 4=0, Index=0, Num=::NewEnumerator3)
       )
      )
     )
    )) {isSetParameters=true}
  ]
  rendererClasses [&NiagaraRibbonRendererProperties]
  renderers [
   {rendererClass=&NiagaraRibbonRendererProperties}
   {rendererClass=&NiagaraRibbonRendererProperties, rendererIndex=1}
  ]
 }
 Tip_Trail {
  emitterSpawnScript [
   /SetVariables(Emitter.ColorInt:NiagaraInt32=&RandomRangeInt(Minimum=0, Maximum=3)) {isSetParameters=true}
   /SetVariables(Emitter.RandomInt_Galaxy:NiagaraInt32=&RandomRangeInt(Minimum=0, Maximum=1)) {isSetParameters=true}
  ]
  emitterUpdateScript [
   &EmitterState(Life Cycle Mode=::NewEnumerator1, Loop Behavior=::NewEnumerator1, Loop Duration=0.27, UseLoopDelay=true)
   &FN_SpawnAnimTrail()
  ]
  particleSpawnScript [
   &FN_InitializeAnimTrailBoneIndecies(First Bone or Socket Index=1, Second Bone or Socket Index=0)
   &FN_InitializeAnimTrail(
    Color=&VariantColorSelector(
     Variant Amount=4
     Original Color=<rgba 0.6892 0.6892 0.6892 1>
     Variant 1 Color=<rgba 0.7205 0.7205 0.7205 1>
     Variant 2 Color=<rgba 1 0.5652 0.1355 1>
     Variant 3 Color=&SelectLinearColorFromArray(
      Array Sampling Mode=::NewEnumerator1
      Direct Array Index=$Emitter.ColorInt
      Color Selection Array=DataInterface() {propertyValues {ColorData [<rgba 2.6304 1.6562 3 1>, <rgba 0.9844 1.37 3 1>, <rgba 1.3281 2.7819 3 1>, <rgba 3 1.4531 2.5023 1>]}}
      Random Seed=?
     )
     Variant 4 Color=<rgba 0.54 1.77 3 1>
     Variant 5 Color=<rgba 0.2865 0.2298 0.4986 1>
    )
    Lifetime=0.22
   )
  ]
  particleUpdateScript [
   &FN_UpdateAnimTrail() {module=FN_UpdateAnimTrail001}
   &SolveForcesAndVelocity(Clamp Velocity=false, Limit Acceleration=false)
   &ScaleColor(
    Scale Alpha=&Multiply_Float(
     A=&FloatFromCurve(
      FloatCurve=DataInterface() {propertyValues {Curve {
         defaultValue=1
         keys [{interpMode=RCIM_Cubic, tangentMode=RCTM_User, time=0, value=1}, {interpMode=RCIM_Cubic, tangentMode=RCTM_User, time=1, value=0}]
        }}}
      Scale Curve=$System.PerSwingFade
     )
     B=&MergeHelper_FloatSwitch(
      0=0.5
      1=1
      2=1
      3=0
      4=1
      5=0
      Index=&MakeIntFromFloat(Float=$System.VariantIndex)
      Num=::NewEnumerator4
     )
    )
    Scale RGB=&Multiply_Vector(
     A=&VectorFromFloat(Value=&FloatFromCurve(
       FloatCurve=DataInterface() {propertyValues {Curve {keys [{interpMode=RCIM_Cubic, tangentMode=RCTM_User, time=0, value=1}, {interpMode=RCIM_Cubic, tangentMode=RCTM_User, time=1, value=0}]}}}
       Scale Curve=$System.TODFloat
      ))
     B=&MakeVectorFromLinearColorRGB(Color=$System.TODColor)
    )
    ScaleA=true
    ScaleRGB=true
   )
  ]
  rendererClasses [&NiagaraRibbonRendererProperties]
  renderers [{rendererClass=&NiagaraRibbonRendererProperties}]
 }
 Variant_Sprites0 {
  emitterUpdateScript [
   &EmitterState(
    Enable Distance Culling=true
    Enable Visibility Culling=true
    MaxDistance=2500
    Module.bApplySpawnCountScaleByDistance=true
    Module.bUseMaxDistance=true
    Module.bUseMinDistance=false
    Scalability Mode=::NewEnumerator1
    Scale Spawn Count=true
   )
   &SpawnPerUnit(
    Module.Use Max Movement Threshold=true
    Module.Use Movement Tolerance=false
    Spawn Spacing:NiagaraFloat=&MakeCustomFloatFromBool(Bool=$System.Galaxy, True Float=40, False Float=0)
    Use Spawn Probability=false
   )
   &SpawnRate(
    SpawnRate=&MakeCustomFloatFromBool(
     Bool=/Niagara/DynamicInputs/Arrays/SelectBoolFromArray(
      Array Sampling Mode:ENiagara_LifetimeMode=::NewEnumerator0
      Direct Array Index:NiagaraInt32=&MakeIntFromFloat(Float=$User.Variant)
      Direct Array Mode:ENiagara_WrapClamp=::NewEnumerator1
      Evaluation Type:ENiagaraRandomnessEvaluation=::NewEnumerator0
      Recalculate Random Each Loop:NiagaraBool=false
      Randomness Mode:ENiagaraRandomnessMode=::NewEnumerator0
      Random Seed:NiagaraInt32=?
      Fixed Random Seed:NiagaraBool=false
      Bool Selection Array:NiagaraDataInterfaceArrayBool=DataInterface() {propertyValues {GpuSyncMode=SyncCpuToGpu, MaxElements=0, BoolData [false,false,true,true,true,false]}}
     )
     True Float=45
     False Float=0
    )
    Use Spawn Probability=false
   )
  ]
  particleSpawnScript [
   &InitializeParticle(
    Color=&MergeHelper_ColorSwitch(
     0=<rgba 0 0 0 0>
     1=<rgba 0 0 0 0>
     2=&ColorFromCurve(
      ColorCurve=DataInterface() {propertyValues {
        AlphaCurve {keys [<at 0 1 RCIM_Cubic>]}
        BlueCurve {keys [<at 0.4274 0 RCIM_Cubic>,<at 0.4699 0 RCIM_Cubic>]}
        GreenCurve {keys [<at 0.4274 0.6232 RCIM_Cubic>,<at 0.4699 0.3648 RCIM_Cubic>]}
        RedCurve {keys [<at 0.4274 1 RCIM_Cubic>,<at 0.4699 1 RCIM_Cubic>]}
       }}
      CurveIndex=&RandomRangeFloat(Minimum=0, Maximum=1)
     )
     3=&ColorFromCurve(
      ColorCurve=DataInterface() {propertyValues {
        AlphaCurve {keys [<at 0 1 RCIM_Cubic>]}
        BlueCurve {keys [<at 0.4274 1 RCIM_Cubic>,<at 0.4699 1 RCIM_Cubic>]}
        GreenCurve {keys [<at 0.4274 0.7649 RCIM_Cubic>,<at 0.4699 0.3664 RCIM_Cubic>]}
        RedCurve {keys [<at 0.4274 0.3886 RCIM_Cubic>,<at 0.4699 0.7831 RCIM_Cubic>]}
       }}
      CurveIndex=&RandomRangeFloat(Minimum=0, Maximum=1)
     )
     4=&ColorFromCurve(
      ColorCurve=DataInterface() {propertyValues {
        AlphaCurve {keys [<at 0 1 RCIM_Cubic>]}
        BlueCurve {keys [<at 0.4274 1 RCIM_Cubic>,<at 0.4699 1 RCIM_Cubic>]}
        GreenCurve {keys [<at 0.4274 0.9469 RCIM_Cubic>,<at 0.4699 0.6167 RCIM_Cubic>]}
        RedCurve {keys [<at 0.4274 0.5208 RCIM_Cubic>,<at 0.4699 0.4375 RCIM_Cubic>]}
       }}
      CurveIndex=&RandomRangeFloat(Minimum=0, Maximum=1)
     )
     5=&ColorFromCurve(
      ColorCurve=DataInterface() {propertyValues {
        AlphaCurve {keys [<at 0 1 RCIM_Cubic>]}
        BlueCurve {keys [<at 0.4274 1 RCIM_Cubic>,<at 0.4699 1 RCIM_Cubic>]}
        GreenCurve {keys [<at 0.4274 0 RCIM_Cubic>,<at 0.4699 0.282 RCIM_Cubic>]}
        RedCurve {keys [<at 0.4274 0.6698 RCIM_Cubic>,<at 0.4699 0 RCIM_Cubic>]}
       }}
      CurveIndex=&RandomRangeFloat(Minimum=0, Maximum=1)
     )
     Index=&MakeIntFromFloat(Float=$User.Variant, Conversion Mode=::NewEnumerator1)
     Num=::NewEnumerator4
    )
    Lifetime=&MergeHelper_FloatSwitch(
     0=0.4
     1=0.4
     2=0.4
     3=0.4
     4=0.4
     5=0.25
     Index=&MakeIntFromFloat(Float=$User.Variant)
     Num=::NewEnumerator4
    )
    Position Offset=&RandomRangeVector(Minimum=<-5 -5 60>, Maximum=<-5 5 20>)
    Position Offset Coordinate Space=::Local
    Ribbon Width=0.75
    Ribbon Width Mode=::NewEnumerator1
    Sprite Size Mode=::NewEnumerator3
    Uniform Sprite Size=&MergeHelper_FloatSwitch(
     0=0
     1=0
     2=&RandomRangeFloat(Minimum=3, Maximum=6)
     3=&RandomRangeFloat(Minimum=5, Maximum=6)
     4=&RandomRangeFloat(Minimum=5, Maximum=7)
     5=&RandomRangeFloat(Minimum=10, Maximum=15)
     Index=&MakeIntFromFloat(Float=$User.Variant)
     Num=::NewEnumerator4
    )
    UsePositionOffset=true
    Write Material Random=true
   )
   &SocketLocation(
    Interpolate Socket Sampling=true
    SamplePosOffset=false
    Socket Reader=$System.Socket Reader
    Socket Sampling Mode=::NewEnumerator3
    VelocitySpeedLimit=false
   )
   &ShapeLocation(MaskSpawnGroup=false, Override Seed=false, Sphere Radius=20, UseOwnerScale=true)
   &AddVelocity(
    Module.Use Falloff Distance=false
    Override Seed=false
    Velocity Mode=::NewEnumerator1
    Velocity Speed=&RandomRangeFloat(Minimum=0, Maximum=25)
   ) {module=AddVelocity001}
   /SetVariables(Particles.RibbonID=/Niagara/DynamicInputs/NiagaraID/MakeNiagaraIDFromInt(Index:NiagaraInt32=1, Acquire Tag:NiagaraInt32=0)) {isSetParameters=true}
  ]
  particleUpdateScript [
   &ParticleState()
   &SpriteRotationRate(Rotation Rate=&FloatFromCurve(
     FloatCurve=DataInterface() {propertyValues {Curve {keys [{interpMode=RCIM_Cubic, tangentMode=RCTM_User, time=0, value=1}, {interpMode=RCIM_Cubic, tangentMode=RCTM_User, time=1, value=0}]}}}
     Scale Curve=&Multiply_FloatByInt(
      Float=&RandomRangeFloat(Minimum=100, Maximum=200)
      Integer=&SelectIntFromArray(
       Array Sampling Mode=::NewEnumerator1
       Direct Array Index=?
       Int Selection Array=DataInterface() {propertyValues {IntData [-1,1]}}
       Random Seed=0
      )
     )
    ))
   &AccelerationForce(Acceleration=<0 0 -100>)
   &Drag(Drag=10, Use Linear Drag=true, Use Rotational Drag=true)
   &CurlNoiseForce(Bias Noise=false, Noise Strength=200, Pan Noise=false)
   &SolveForcesAndVelocity(Clamp Velocity=false, Limit Acceleration=false)
   &ScaleSpriteSize(Uniform Curve Sprite Scale=DataInterface() {propertyValues {Curve {keys [
        {arriveTangent=-1.3862, interpMode=RCIM_Cubic, time=0, value=1}
        {arriveTangent=-1.8417, interpMode=RCIM_Cubic, leaveTangent=-1.8417, tangentMode=RCTM_User, time=1, value=0}
       ]}}})
   &ScaleColor(
    Scale Alpha=&Multiply_Float(
     A=&FloatFromCurve(FloatCurve=DataInterface() {propertyValues {Curve {keys [<at 0.3 1>,<at 1 0>]}}})
     B=&FNTimeOfDayScalar(Night Scale=0.5, Morning Scale=0.75, Evening Scale=0.75)
    )
    Scale RGB=&MakeVectorFromLinearColorRGB(Color=$System.TODColor)
    ScaleA=true
    ScaleRGB=true
   )
   /SetVariables(
    Particles.GlowColor=&MakeLinearColorFromVectorAndFloat(
     "Vector (RGB)"=&Multiply_Vector(
      A=&MakeVectorFromLinearColorRGB(Color=&MergeHelper_ColorSwitch(
        0=&Multiply_LinearColorByFloat(
         Linear Color=&ColorFromCurve(
          ColorCurve=DataInterface() {propertyValues {
            AlphaCurve {keys [<at 0 1 RCIM_Cubic>]}
            BlueCurve {keys [
              <at 0 1>
              <at 0.2788 1>
              {arriveTangent=0.4199, interpMode=RCIM_Cubic, leaveTangent=0.4199, time=0.5087, value=0.9509}
              <at 0.7781 1.3267>
              <at 1 1>
             ]}
            GreenCurve {keys [
              <at 0 0.6539>
              <at 0.2788 0.2522>
              {arriveTangent=1.0874, interpMode=RCIM_Cubic, leaveTangent=1.0874, time=0.5087, value=0.5156}
              <at 0.7781 1.5>
              <at 1 0.6539>
             ]}
            RedCurve {keys [
              <at 0 0.4301>
              <at 0.2788 0.613>
              {arriveTangent=-0.1655, interpMode=RCIM_Cubic, leaveTangent=-0.1655, time=0.5087, value=1}
              <at 0.7781 0.3013>
              <at 1 0.4301>
             ]}
           }}
          CurveIndex=&Frac_Float(Float=&Add_Float(A=$Particles.NormalizedAge, B=&RandomRangeFloat(Minimum=0, Maximum=-0.5)))
         )
         Float=&RandomRangeFloat(Minimum=0.2, Maximum=0.5)
        )
        1=&Multiply_LinearColorByFloat(
         Linear Color=&ColorFromCurve(
          ColorCurve=DataInterface() {propertyValues {
            AlphaCurve {keys [<at 0 1 RCIM_Cubic>]}
            BlueCurve {keys [
              <at 0.0284 0.8677>
              {arriveTangent=-0.4812, interpMode=RCIM_Cubic, leaveTangent=-0.4812, time=0.4737, value=0.8646}
              <at 0.7925 0.5>
              <at 0.9918 0.8646>
             ]}
            GreenCurve {keys [
              <at 0.0284 1>
              {arriveTangent=-1.2221, interpMode=RCIM_Cubic, leaveTangent=-1.2221, time=0.4737, value=0.1067}
              <at 0.7925 0.0662>
              <at 0.9918 0.7183>
             ]}
            RedCurve {keys [
              <at 0.0284 0>
              {arriveTangent=0.4161, interpMode=RCIM_Cubic, leaveTangent=0.4161, time=0.4737, value=0.231}
              <at 0.7925 0.3179>
              <at 0.9918 0>
             ]}
           }}
          CurveIndex=&Frac_Float(Float=&Add_Float(A=$Particles.NormalizedAge, B=&RandomRangeFloat(Minimum=0, Maximum=1)))
         )
         Float=&RandomRangeFloat(Minimum=0.2, Maximum=0.5)
        )
        2=?
        3=?
        Index=&MergeHelper_IntSwitch(0=0, 1=0, 2=0, 3=0, 4=0, 5=1, Index=&MakeIntFromFloat(Float=$User.Variant), Num=::NewEnumerator4)
        Num=::NewEnumerator0
       ))
      B=&MakeVectorFromLinearColorRGB(Color=$System.TODColor)
     )
     "Float (Alpha)"=&FloatFromCurve(
      FloatCurve=DataInterface() {propertyValues {Curve {
         defaultValue=1
         keys [{interpMode=RCIM_Cubic, tangentMode=RCTM_User, time=0, value=1}, {interpMode=RCIM_Cubic, tangentMode=RCTM_User, time=0.8, value=0}]
        }}}
      Scale Curve=&Multiply_Float(A=0.5, B=$System.TODFloat)
     )
    )
    Particles.RibbonColor:LinearColor=&MakeLinearColorFromVectorAndFloat(
     "Vector (RGB)"=&Multiply_Vector(
      A=&MakeVectorFromLinearColorRGB(Color=&Multiply_LinearColorByFloat(
        Linear Color=&ColorFromCurve(
         ColorCurve=DataInterface() {propertyValues {
           AlphaCurve {keys [<at 0 1 RCIM_Cubic>]}
           BlueCurve {keys [<at 0.0378 0.8677>, {arriveTangent=-0.0033, interpMode=RCIM_Cubic, time=0.3642, value=0.8646}, <at 0.7623 0.5>, <at 0.9918 0.8646>]}
           GreenCurve {keys [<at 0.0378 1>, {arriveTangent=-0.2924, interpMode=RCIM_Cubic, time=0.3642, value=0.1067}, <at 0.7623 0.0662>, <at 0.9918 0.7183>]}
           RedCurve {keys [<at 0.0378 0>,<at 0.3642 0.231 RCIM_Cubic>,<at 0.7623 0.3179>,<at 0.9918 0>]}
          }}
         CurveIndex=&Frac_Float(Float=&Add_Float(
           A=&Multiply_Float(A=$Particles.NormalizedAge, B=0.3)
           B=&RandomRangeFloat(Minimum=0, Maximum=1)
          ))
        )
        Float=&RandomRangeFloat(Minimum=0.6, Maximum=1)
       ))
      B=&MakeVectorFromLinearColorRGB(Color=$System.TODColor)
     )
     "Float (Alpha)"=&FloatFromCurve(
      FloatCurve=DataInterface() {propertyValues {Curve {
         defaultValue=1
         keys [{interpMode=RCIM_Cubic, tangentMode=RCTM_User, time=0, value=1}, {interpMode=RCIM_Cubic, tangentMode=RCTM_User, time=0.8, value=0}]
        }}}
      Scale Curve=&Multiply_Float(A=1.5, B=$System.TODFloat)
     )
    )
   ) {isSetParameters=true}
  ]
  rendererClasses [&NiagaraSpriteRendererProperties, &NiagaraRibbonRendererProperties]
  renderers [
   {rendererClass=&NiagaraSpriteRendererProperties}
   {rendererClass=&NiagaraSpriteRendererProperties, rendererIndex=1}
   {rendererClass=&NiagaraSpriteRendererProperties, rendererIndex=2}
   {rendererClass=&NiagaraRibbonRendererProperties, rendererIndex=3}
   {rendererClass=&NiagaraSpriteRendererProperties, rendererIndex=4}
  ]
 }
 Variant_Trails {
  emitterUpdateScript [
   &EmitterState(Life Cycle Mode=::NewEnumerator1, Loop Behavior=::NewEnumerator1, Loop Duration=0.27, UseLoopDelay=true)
   &FN_SpawnAnimTrail()
  ]
  particleSpawnScript [
   &FN_InitializeAnimTrailBoneIndecies()
   &FN_InitializeAnimTrail(
    AnimTrailWidth=&MergeHelper_FloatSwitch(
     0=0
     1=0
     2=0
     3=1.5
     4=0
     5=1
     Index=&MakeIntFromFloat(Float=$System.VariantIndex)
     Num=::NewEnumerator4
    )
    Lifetime=0.27
   )
   &DynamicMaterialParameters(
    Index 0 Param 2=1
    Index 0 Param 3=5
    Index 0 Param 4=0.3
    Index 1 Param 1=0.1
    Index 1 Param 2=0.5
    Index 1 Param 3=0.1
    Index 2 Param 1=&FloatFromCurve(
     FloatCurve=DataInterface() {propertyValues {Curve {keys [{arriveTangent=-0.4165, interpMode=RCIM_Cubic, time=0.1, value=-0.875}, {arriveTangent=12.7593, leaveTangent=12.7593, time=0.3, value=0.5}]}}}
     CurveIndex=$Emitter.Age
    )
    Module.Index0Float0Write=true
    Module.Index0Float1Write=true
    Module.Index0Float2Write=true
    Module.Index0Float3Write=true
    Module.Index1Float0Write=true
    Module.Index1Float1Write=true
    Module.Index1Float2Write=true
    Module.Index1Float3Write=true
    Module.Index2Float0Write=true
    Module.Index2Float1Write=true
    Module.Index2Float2Write=true
    Module.Index2Float3Write=true
    Write Parameter Index 1=::NewEnumerator1
    Write Parameter Index 2=::NewEnumerator1
   )
  ]
  particleUpdateScript [
   &FN_UpdateAnimTrail() {module=FN_UpdateAnimTrail001}
   &ScaleColor(
    Scale Alpha=&Multiply_Float(
     A=&FloatFromCurve(
      FloatCurve=DataInterface() {propertyValues {Curve {
         defaultValue=1
         keys [
          {interpMode=RCIM_Cubic, tangentMode=RCTM_User, time=0, value=1}
          {arriveTangent=-2.3945, interpMode=RCIM_Cubic, leaveTangent=-2.3945, tangentMode=RCTM_User, time=1.001, value=0}
         ]
        }}}
      Scale Curve=$System.PerSwingFade
     )
     B=&MergeHelper_FloatSwitch(
      0=0
      1=0
      2=0
      3=0.2
      4=0
      5=0.4
      Index=&MakeIntFromFloat(Float=$System.VariantIndex)
      Num=::NewEnumerator4
     )
    )
    Scale RGB=&Multiply_Vector(
     A=&VectorFromFloat(Value=&FloatFromCurve(
       FloatCurve=DataInterface() {propertyValues {Curve {keys [{interpMode=RCIM_Cubic, tangentMode=RCTM_User, time=0, value=1}, {interpMode=RCIM_Cubic, tangentMode=RCTM_User, time=1, value=0}]}}}
       Scale Curve=$System.TODFloat
      ))
     B=&MakeVectorFromLinearColorRGB(Color=$System.TODColor)
    )
    ScaleA=true
    ScaleRGB=true
   ) {module=ScaleColor002}
   &ScaleRibbonWidth(Ribbon Width Scale=&MergeHelper_FloatSwitch(
     0=0
     1=0
     2=0
     3=&FloatFromCurve(FloatCurve=DataInterface() {propertyValues {Curve {keys [<at 0 1>,<at 1 0>]}}})
     4=0
     5=1
     Index=&MakeIntFromFloat(Float=$System.VariantIndex)
     Num=::NewEnumerator4
    ))
   &SolveForcesAndVelocity(Clamp Velocity=false, Limit Acceleration=false)
  ]
  rendererClasses [&NiagaraRibbonRendererProperties]
  renderers [
   {rendererClass=&NiagaraRibbonRendererProperties}
   {rendererClass=&NiagaraRibbonRendererProperties, rendererIndex=1}
  ]
 }
}
summary {
 systemName=NS_Pickaxe_SkelFlask_Trail
 emitters [
  {emitterName=Variant_Trails, rendererClasses [&NiagaraRibbonRendererProperties]}
  {emitterName=Tip_Trail, rendererClasses [&NiagaraRibbonRendererProperties]}
  {emitterName=Main_Trail, rendererClasses [&NiagaraRibbonRendererProperties]}
  {emitterName=Glow_Trail, rendererClasses [&NiagaraRibbonRendererProperties]}
  {
   emitterName=Variant_Sprites0
   rendererClasses [&NiagaraSpriteRendererProperties, &NiagaraRibbonRendererProperties]
  }
 ]
 userVariables [
  User.bIsFrontEnd {type=NiagaraBool}
  User.bIsFrontEndPreview {type=NiagaraBool}
  User.RibbonIDOffset {type=NiagaraInt32}
  User.SwingSectionIndex {type=NiagaraInt32}
  User.Variant {type=NiagaraFloat, defaultValue=0, description=""}
 ]
}
